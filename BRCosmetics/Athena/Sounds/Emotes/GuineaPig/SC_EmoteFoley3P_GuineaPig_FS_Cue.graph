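# Sound cue for third-person footstep foley in the GuineaPig emote.
# It randomly picks one of ten equally weighted step waves, adding slight pitch and volume variation.
# An outer continuous modulator reads EmoteMoveSpeed, but equal output limits keep its volume and pitch effect constant.
# It uses emote attenuation and routes as emote music, with that sound class, music concurrency and an audio-analysis submix send.
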SoundCue /BRCosmetics/Athena/Sounds/Emotes/GuineaPig/SC_EmoteFoley3P_GuineaPig_FS_Cue
root=ModulatorContinuous_0
nodes [
 {
  id=ModulatorContinuous_0
  children [Modulator_0]
  properties {
   pitchModulationParams {minOutput=1, parameterName=EmoteMoveSpeed}
   volumeModulationParams {maxOutput=0.7, minOutput=0.7, parameterName=EmoteMoveSpeed}
  }
 }
 {id=Modulator_0, children [Random_1], properties {pitchMax=1.02, pitchMin=0.98, volumeMax=1.02, volumeMin=0.98}}
 {
  id=Random_1
  children [WavePlayer_1, WavePlayer_2, WavePlayer_3, WavePlayer_4, WavePlayer_5, WavePlayer_6, WavePlayer_7, WavePlayer_8, WavePlayer_9, WavePlayer_10]
  properties {weights [1,1,1,1,1,1,1,1,1,1]}
 }
 {id=WavePlayer_1, properties {soundWaveAssetPtr=/Game/Athena/Sounds/Emotes/GuineaPig/IP_Emote_GuineaPig_Step_01}}
 {id=WavePlayer_10, properties {soundWaveAssetPtr=/Game/Athena/Sounds/Emotes/GuineaPig/IP_Emote_GuineaPig_Step_10}}
 {id=WavePlayer_2, properties {soundWaveAssetPtr=/Game/Athena/Sounds/Emotes/GuineaPig/IP_Emote_GuineaPig_Step_02}}
 {id=WavePlayer_3, properties {soundWaveAssetPtr=/Game/Athena/Sounds/Emotes/GuineaPig/IP_Emote_GuineaPig_Step_03}}
 {id=WavePlayer_4, properties {soundWaveAssetPtr=/Game/Athena/Sounds/Emotes/GuineaPig/IP_Emote_GuineaPig_Step_04}}
 {id=WavePlayer_5, properties {soundWaveAssetPtr=/Game/Athena/Sounds/Emotes/GuineaPig/IP_Emote_GuineaPig_Step_05}}
 {id=WavePlayer_6, properties {soundWaveAssetPtr=/Game/Athena/Sounds/Emotes/GuineaPig/IP_Emote_GuineaPig_Step_06}}
 {id=WavePlayer_7, properties {soundWaveAssetPtr=/Game/Athena/Sounds/Emotes/GuineaPig/IP_Emote_GuineaPig_Step_07}}
 {id=WavePlayer_8, properties {soundWaveAssetPtr=/Game/Athena/Sounds/Emotes/GuineaPig/IP_Emote_GuineaPig_Step_08}}
 {id=WavePlayer_9, properties {soundWaveAssetPtr=/Game/Athena/Sounds/Emotes/GuineaPig/IP_Emote_GuineaPig_Step_09}}
]
properties {
 attenuationSettings=&Fort_Emotes_Athena_Att
 maxDistance=1600
 soundClassObject=/Game/Sounds/FortSoundClassesAndModes/Fort_Music_Emote_3P
 virtualizationMode=PlayWhenSilent
 volumeMultiplier=0.65
 concurrencyOverrides {maxCount=3}
 concurrencySet [/Game/Sounds/Concurrency_Rules/Athena_EmoteMusic_Conc]
 soundSubmixSends [{minSendLevel=0, soundSubmix=&Fortnite_AudioAnalysis_Music}]
}
