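# Looping GuineaPig emote sound cue, named as foley but routed through emote music class and concurrency, with emote attenuation.
# A mixer layers a constant base loop with Idle and Fwd loops, the Idle input mixed lower.
# Continuous modulators driven by EmoteMoveSpeed vary volume on Idle and Fwd, plus a slight pitch shift on Fwd.
# A submix send feeds music audio analysis.
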SoundCue /BRCosmetics/Athena/Sounds/Emotes/GuineaPig/SC_EmoteFoley_GuineaPig_LP_Cue
root=Mixer_0
nodes [
 {id=Mixer_0, children [WavePlayer_0, ModulatorContinuous_2, ModulatorContinuous_0], properties {inputVolume [1,0.6,1]}}
 {
  id=ModulatorContinuous_0
  children [WavePlayer_1]
  properties {
   pitchModulationParams {maxOutput=0.98, minOutput=1, parameterName=EmoteMoveSpeed}
   volumeModulationParams {maxOutput=0.8, minOutput=0.5, parameterName=EmoteMoveSpeed}
  }
 }
 {
  id=ModulatorContinuous_2
  children [WavePlayer_2]
  properties {
   pitchModulationParams {minOutput=1, parameterName=EmoteMoveSpeed}
   volumeModulationParams {maxOutput=0.8, minOutput=0.1, parameterName=EmoteMoveSpeed}
  }
 }
 {id=WavePlayer_0, properties {bLooping=true, soundWaveAssetPtr=/Game/Athena/Sounds/Emotes/GuineaPig/IP_Emote_GuineaPig_LP}}
 {id=WavePlayer_1, properties {bLooping=true, soundWaveAssetPtr=/Game/Athena/Sounds/Emotes/GuineaPig/IP_Emote_GuineaPig_Fwd_LP}}
 {id=WavePlayer_2, properties {bLooping=true, soundWaveAssetPtr=/Game/Athena/Sounds/Emotes/GuineaPig/IP_Emote_GuineaPig_Idle_LP}}
]
properties {
 attenuationSettings=&Fort_Emotes_Athena_Att
 maxDistance=1600
 soundClassObject=&Fort_Music_Emote
 volumeMultiplier=1
 concurrencyOverrides {maxCount=3}
 concurrencySet [/Game/Sounds/Concurrency_Rules/Athena_EmoteMusic_Conc]
 soundSubmixSends [{minSendLevel=0, soundSubmix=&Fortnite_AudioAnalysis_Music}]
}
